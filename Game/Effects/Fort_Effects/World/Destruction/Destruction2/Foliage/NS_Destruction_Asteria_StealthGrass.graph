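# Niagara destruction effect for stealth grass that bursts user-sized clouds of mesh leaves, pushed along the hit vector.
# Leaves fall under wind and aerodynamic drag; the GPU emitter adds collision, while Leaves_Mobile is a CPU variant without it.
# User.bIsCurie enables orange curl-noise sparks, spawned from GPU leaves via particle reads and burst within bounds.
# The same flag drives leaf dynamic material parameters, one fading over particle life.
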NiagaraSystem /Game/Effects/Fort_Effects/World/Destruction/Destruction2/Foliage/NS_Destruction_Asteria_StealthGrass
dependencies {
 usedDataInterfaces [
  &NiagaraDataInterfaceMeshRendererInfo
  &NiagaraDataInterfaceCurve
  &NiagaraDataInterfaceCollisionQuery
  /Script/ChaosNiagara.NiagaraDataInterfacePhysicsField
  &NiagaraDataInterfaceLandscape
  &NiagaraDataInterfaceParticleRead
 ]
 usedDynamicInputs [
  &self:ScaleFactor
  &ClampFloat
  &Divide_Float
 ]
 usedModules [
  &self:SystemSpawnScript.SystemScriptSource.SystemScriptGraph.NiagaraNodeAssignment_0.SetVariables_#1
  &SystemState
  &self:SystemSpawnScript.SystemScriptSource.SystemScriptGraph.NiagaraNodeAssignment_1.SetVariables_#2
 ]
}
emitters {
 Curie_Sparks {
  simTarget=GPUComputeSim
  emitterUpdateScript [
   &EmitterState(Life Cycle Mode=::NewEnumerator1, Loop Behavior=::NewEnumerator1, Loop Duration=0.3, UseLoopDelay=false)
   &SpawnParticlesFromOtherEmitter(
    Attribute Reader=DataInterface() {propertyValues {EmitterBinding {bindingMode=Other, emitterName=Leaves}}}
    Spawn Rate=&MakeCustomFloatFromBool(Bool=$User.bIsCurie, True Float=10, False Float=0)
    Spawn Rate Per Particle Cap=10
    Use Maximum Age=false
    Use Minimum Age=false
   )
  ]
  particleSpawnScript [
   &InitializeParticle(
    Color=<rgba 30 13.125 3 1>
    Lifetime Max=1
    Lifetime Min=0.15
    Lifetime Mode=::NewEnumerator1
    Sprite Size Max=<3.5 2>
    Sprite Size Min=<2.5 1>
    Sprite Size Mode=::NewEnumerator2
    UsePositionOffset=false
    Write Material Random=true
   )
   &SampleParticlesFromOtherEmitter(Velocity Application:ENiagara_DirectReadApplicationMode=::NewEnumerator0, Velocity Sampling=::NewEnumerator0)
  ]
  particleUpdateScript [
   &ParticleState()
   &GravityForce()
   /Niagara/Modules/Update/Forces/CurlNoiseForce(
    Noise Frequency=300
    Noise Strength=&Multiply_Float(
     A=$Particles.MaterialRandom
     B=&FloatFromCurve(
      FloatCurve=DataInterface() {propertyValues {Curve {keys [
          <at 0 4147.9209 RCIM_Cubic>
          {arriveTangent=998.7806, interpMode=RCIM_Cubic, leaveTangent=998.7806, time=0.897, value=2308.9619}
          {arriveTangent=839.8494, interpMode=RCIM_Cubic, leaveTangent=839.8494, time=2.2666, value=6411.7881}
          {arriveTangent=818.7506, interpMode=RCIM_Cubic, leaveTangent=818.7506, time=3.6727, value=4640.1284}
          {arriveTangent=-197.3881, interpMode=RCIM_Cubic, leaveTangent=-197.3881, time=4.7165, value=8417.627}
          {arriveTangent=-812.5138, interpMode=RCIM_Cubic, leaveTangent=-812.5138, time=6.6551, value=4051.4272}
          {arriveTangent=379.0706, interpMode=RCIM_Cubic, leaveTangent=379.0706, time=9.3392, value=4661.6182}
          {arriveTangent=-88.1153, interpMode=RCIM_Cubic, leaveTangent=-88.1153, time=11.8318, value=6013.7651}
          <at 15.169 4147.9209 RCIM_Cubic>
         ]}}}
      CurveIndex=&Add_Float(A=$Engine.System.Age, B=$Particles.Age)
      Scale Curve=0.7
     )
    )
    Pan Noise=true
    Pan Noise Field=<0 0 1000>
   )
   &Collision(
    Age Colliding Particles=false
    Friction=0.35
    GPU Collision Type=::NewEnumerator2
    Module.Control Rest State Via Surface Normal=false
    Module.EnableMaxCollisionCount=false
    Module.Randomize Collision Normal Vector=false
    Restitution=0.4
   )
   &Drag(Drag=0.8, Use Linear Drag=true, Use Rotational Drag=false)
   &SolveForcesAndVelocity(Clamp Velocity=false, Limit Acceleration=false)
   &ScaleSpriteSizeBySpeed(Max Scale Factor=<1 10>, Velocity Threshold=1300)
   &ScaleColor(
    Scale Alpha=&FloatFromCurve(FloatCurve=DataInterface() {propertyValues {Curve {keys [<at 0 0>,<at 0.1997 0.9273>,<at 1.0011 0.002>]}}})
    ScaleA=true
    ScaleRGB=true
   )
  ]
  rendererClasses [&NiagaraSpriteRendererProperties]
  renderers [{rendererClass=&NiagaraSpriteRendererProperties}]
 }
 Curie_Sparks001 {
  emitterUpdateScript [
   &EmitterState(Life Cycle Mode=::NewEnumerator1, Loop Behavior=::NewEnumerator1, Loop Duration=0.3, UseLoopDelay=false)
   &SpawnBurst_Instantaneous(
    Spawn Count=&MakeIntFromFloat(Float=&Multiply_Float(
      A=&Multiply_Float(A=50, B=$System.ObjectRadiusScaleFactor)
      B=&MakeCustomFloatFromBool(Bool=$User.bIsCurie, True Float=1, False Float=0)
     ))
    Use Spawn Probability=false
    UseLoopCountLimit=false
   )
  ]
  particleSpawnScript [
   &InitializeParticle(
    Color=<rgba 30 13.125 3 1>
    Lifetime Max=1
    Lifetime Min=0.15
    Lifetime Mode=::NewEnumerator1
    Sprite Size Max=<3.5 2>
    Sprite Size Min=<2.5 1>
    Sprite Size Mode=::NewEnumerator2
    UsePositionOffset=false
    Write Material Random=true
   )
   &ShapeLocation(
    Box Midpoint=<0.5 0.5 0>
    Box Size=$User.Bounds
    MaskSpawnGroup=false
    Override Seed=false
    Shape Primitive=::NewEnumerator2
    Surface Only=false
    UseOwnerScale=true
   )
   /Niagara/Modules/Spawn/Location/SphereLocation(MaskSpawnGroup=false, Module.Override Randomness=false, Sphere Radius=5, Surface Only=false)
   &PointForce(
    Force Falloff Distance:NiagaraFloat=$User.Radius
    Force Origin:NiagaraPosition=$User.HitLocation001
    Force Strength=&RandomRangeFloat(Minimum=750, Maximum=1000)
    Module.Use Falloff=true
    Module.UseLinearForce=false
   )
   &LinearForce(
    Coordinate Space:ENiagaraCoordinateSpace=::World
    Force=&Multiply_VectorByFloat(
     Vector=$User.HitVector
     Float=&MakeCustomFloatFromBool(Bool=$User.Detached, True Float=0, False Float=&RandomRangeFloat(Minimum=475, Maximum=775))
    )
   )
   &ApplyInitialForces()
  ]
  particleUpdateScript [
   &ParticleState()
   &GravityForce()
   &CurlNoiseForce(
    Noise Frequency=300
    Noise Strength=&Multiply_Float(
     A=$Particles.MaterialRandom
     B=&FloatFromCurve(
      FloatCurve=DataInterface() {propertyValues {Curve {keys [
          <at 0 4147.9209 RCIM_Cubic>
          {arriveTangent=998.7806, interpMode=RCIM_Cubic, leaveTangent=998.7806, time=0.897, value=2308.9619}
          {arriveTangent=839.8494, interpMode=RCIM_Cubic, leaveTangent=839.8494, time=2.2666, value=6411.7881}
          {arriveTangent=818.7506, interpMode=RCIM_Cubic, leaveTangent=818.7506, time=3.6727, value=4640.1284}
          {arriveTangent=-197.3881, interpMode=RCIM_Cubic, leaveTangent=-197.3881, time=4.7165, value=8417.627}
          {arriveTangent=-812.5138, interpMode=RCIM_Cubic, leaveTangent=-812.5138, time=6.6551, value=4051.4272}
          {arriveTangent=379.0706, interpMode=RCIM_Cubic, leaveTangent=379.0706, time=9.3392, value=4661.6182}
          {arriveTangent=-88.1153, interpMode=RCIM_Cubic, leaveTangent=-88.1153, time=11.8318, value=6013.7651}
          <at 15.169 4147.9209 RCIM_Cubic>
         ]}}}
      CurveIndex=&Add_Float(A=$Engine.System.Age, B=$Particles.Age)
      Scale Curve=0.7
     )
    )
    Pan Noise=true
    Pan Noise Field=<0 0 1000>
   )
   &Collision(
    Age Colliding Particles=false
    Friction=0.35
    Module.Control Rest State Via Surface Normal=false
    Module.EnableMaxCollisionCount=false
    Module.Randomize Collision Normal Vector=false
    Restitution=0.4
   )
   &Drag(Drag=0.8, Use Linear Drag=true, Use Rotational Drag=false)
   &SolveForcesAndVelocity(Clamp Velocity=false, Limit Acceleration=false)
   &ScaleSpriteSizeBySpeed(Max Scale Factor=<1 10>, Velocity Threshold=1300)
   &ScaleColor(
    Scale Alpha=&FloatFromCurve(FloatCurve=DataInterface() {propertyValues {Curve {keys [<at 0 0>,<at 0.1997 0.9273>,<at 1.0011 0.002>]}}})
    ScaleA=true
    ScaleRGB=true
   )
  ]
  rendererClasses [&NiagaraSpriteRendererProperties]
  renderers [{rendererClass=&NiagaraSpriteRendererProperties}]
 }
 Leaves {
  simTarget=GPUComputeSim
  emitterUpdateScript [
   &EmitterState()
   &SpawnBurst_Instantaneous(
    Spawn Count=&MakeIntFromFloat(Float=$User.LeafAmount)
    Use Spawn Probability=false
    UseLoopCountLimit=false
   )
  ]
  particleSpawnScript [
   &InitializeParticle(
    Color Mode=::NewEnumerator0
    Lifetime Max=2.5
    Lifetime Min=1.5
    Lifetime Mode=::NewEnumerator1
    Mesh Renderer Array Visibility Mode=::NewEnumerator3
    Mesh Renderer Info=DataInterface() {propertyValues {MeshRenderer=&self:Leaves_1.NiagaraMeshRendererProperties_0}}
    Mesh Scale Max=<20 10 10>
    Mesh Scale Min=<10 5 5>
    Mesh Scale Mode=::NewEnumerator2
    UsePositionOffset=false
    Write Material Random=true
   )
   &ShapeLocation(
    Cylinder Height=$User.TreeCylinderHeight
    Cylinder Height Midpoint=0
    Cylinder Radius=$User.TreeCylinderRadius
    MaskSpawnGroup=false
    Offset=$User.TreeCylinderOffset
    Offset Mode=::NewEnumerator0
    Override Seed=false
    Shape Primitive=::NewEnumerator1
    Surface Only=false
    UseOwnerScale=true
   )
   &ShapeLocation(MaskSpawnGroup=false, Override Seed=false, Sphere Radius=$User.TreeSpheriodRadius, UseOwnerScale=true) {module=ShapeLocation001}
   &InitialMeshOrientation(Use Rotation Vector=false)
   &PointForce(
    Force Origin=&AddVectorToPosition(Position=/Niagara/DynamicInputs/Helpers/SimulationPosition(), Vector=<0 0 1000>)
    Force Strength=&RandomRangeFloat(Minimum=333, Maximum=1000)
    Module.Use Falloff=false
    Module.UseLinearForce=false
   )
   &LinearForce(Force=&Multiply_VectorByFloat(Vector=$User.HitVector, Float=&RandomRangeFloat(Minimum=1500, Maximum=1750)))
   &ApplyInitialForces()
  ]
  particleUpdateScript [
   &ParticleState()
   &ScaleMeshSize(Scale Factor=&VectorFromFloat(Value=&FloatFromCurve(FloatCurve=DataInterface() {propertyValues {Curve {keys [<at 0.8 1 RCIM_Cubic>,<at 1 0 RCIM_Cubic>]}}})))
   &Collision(
    Age Colliding Particles=false
    Collision Enabled=&SetBoolByFloatComparison(A=$Particles.Age, B=0.1)
    Control Roll On Collision=true
    Friction=0.333
    Friction During a Bounce=0.3
    GPU Collision Type=::NewEnumerator2
    Mesh Dimensions:Vector3f=<8 3 0>
    Module.Control Rest State Via Surface Normal=false
    Module.EnableMaxCollisionCount=false
    Module.Randomize Collision Normal Vector=true
    Radius Calculation Type=::NewEnumerator1
    Restitution=0
    ScaleRateOfRotation=false
   )
   &GravityForce()
   &WindForce(
    Ground Mask Elevation Mode:ENiagaraWindGroundMaskMode=::NewEnumerator1
    Use Ground Mask=true
    Use Physics Field=true
    Wind Speed=<1000 0 100>
    Wind Speed Scale=&MakeCustomFloatFromBool(Bool=$StackContext.HasCollided, True Float=0, False Float=0.1)
   )
   &AerodynamicDrag(
    Aerodynamic Drag=0.2
    Aerodynamic Rotational Drag=0.05
    Drag Shape Curve=DataInterface() {propertyValues {Curve {keys [
        {interpMode=RCIM_Cubic, tangentMode=RCTM_User, time=0, value=0.1}
        {arriveTangent=2.0799, interpMode=RCIM_Cubic, leaveTangent=2.0799, tangentMode=RCTM_User, time=1, value=1}
       ]}}}
    Lift Shape Curve=DataInterface() {propertyValues {Curve {keys [
        {interpMode=RCIM_Cubic, tangentMode=RCTM_User, time=0, value=0}
        {interpMode=RCIM_Cubic, tangentMode=RCTM_User, time=0.5, value=1}
        {tangentMode=RCTM_User, time=1, value=0}
       ]}}}
    Rotation Strength:NiagaraFloat=&RandomRangeFloat(Minimum=1, Maximum=4)
    Use Linear Drag=true
    Use Rotational Drag=true
   )
   &SolveForcesAndVelocity(Clamp Velocity=false, Limit Acceleration=false)
   &DynamicMaterialParameters(
    Index 0 Param 1=&MakeCustomFloatFromBool(
     Bool=$User.bIsCurie
     True Float=&FloatFromCurve(FloatCurve=DataInterface() {propertyValues {Curve {keys [
          {arriveTangent=-9.3075, interpMode=RCIM_Cubic, leaveTangent=-9.3075, tangentMode=RCTM_User, time=0, value=2}
          {arriveTangent=-2.0574, interpMode=RCIM_Cubic, leaveTangent=-2.0574, tangentMode=RCTM_User, time=0.2, value=0.625}
          <at 0.75 0 RCIM_Cubic>
         ]}}})
     False Float=0
    )
    Index 0 Param 2=&MakeCustomFloatFromBool(Bool=$User.bIsCurie, True Float=1, False Float=0)
    Module.Param0WriteEnabled=true
    Module.Param1WriteEnabled=true
    Module.Param2WriteEnabled=false
    Module.Param3WriteEnabled=false
   )
  ]
  rendererClasses [&NiagaraMeshRendererProperties]
  renderers [{rendererClass=&NiagaraMeshRendererProperties, rendererIndex=0}]
 }
 Leaves_Mobile {
  emitterUpdateScript [
   &EmitterState()
   &SpawnBurst_Instantaneous(
    Spawn Count=&MakeIntFromFloat(Float=$User.LeafAmount)
    Use Spawn Probability=false
    UseLoopCountLimit=false
   )
  ]
  particleSpawnScript [
   &InitializeParticle(
    Color Mode=::NewEnumerator0
    Lifetime Max=2.5
    Lifetime Min=1.5
    Lifetime Mode=::NewEnumerator1
    Mesh Renderer Array Visibility Mode=::NewEnumerator3
    Mesh Renderer Info=DataInterface() {propertyValues {MeshRenderer=&self:Leaves_Mobile_2.NiagaraMeshRendererProperties_0}}
    Mesh Scale Max=<20 10 10>
    Mesh Scale Min=<10 5 5>
    Mesh Scale Mode=::NewEnumerator2
    UsePositionOffset=false
    Write Material Random=true
   )
   &ShapeLocation(
    Cylinder Height=$User.TreeCylinderHeight
    Cylinder Height Midpoint=0
    Cylinder Radius=$User.TreeCylinderRadius
    MaskSpawnGroup=false
    Offset=$User.TreeCylinderOffset
    Offset Mode=::NewEnumerator0
    Override Seed=false
    Shape Primitive=::NewEnumerator1
    Surface Only=false
    UseOwnerScale=true
   )
   &ShapeLocation(MaskSpawnGroup=false, Override Seed=false, Sphere Radius=$User.TreeSpheriodRadius, UseOwnerScale=true) {module=ShapeLocation001}
   &InitialMeshOrientation(Use Rotation Vector=false)
   &PointForce(
    Force Origin=&AddVectorToPosition(Position=&SimulationPosition(), Vector=<0 0 1000>)
    Force Strength=&RandomRangeFloat(Minimum=333, Maximum=1000)
    Module.Use Falloff=false
    Module.UseLinearForce=false
   )
   &LinearForce(Force=&Multiply_VectorByFloat(Vector=$User.HitVector, Float=&RandomRangeFloat(Minimum=1500, Maximum=1750)))
   &ApplyInitialForces()
  ]
  particleUpdateScript [
   &ParticleState()
   &ScaleMeshSize(Scale Factor=&VectorFromFloat(Value=&FloatFromCurve(FloatCurve=DataInterface() {propertyValues {Curve {keys [<at 0.8 1 RCIM_Cubic>,<at 1 0 RCIM_Cubic>]}}})))
   &GravityForce()
   &WindForce(
    Ground Mask Elevation Mode:ENiagaraWindGroundMaskMode=::NewEnumerator1
    Use Ground Mask=true
    Use Physics Field=true
    Wind Speed=<1000 0 100>
    Wind Speed Scale=0.1
   )
   &AerodynamicDrag(
    Aerodynamic Drag=0.2
    Aerodynamic Rotational Drag=0.05
    Drag Shape Curve=DataInterface() {propertyValues {Curve {keys [
        {interpMode=RCIM_Cubic, tangentMode=RCTM_User, time=0, value=0.1}
        {arriveTangent=2.0799, interpMode=RCIM_Cubic, leaveTangent=2.0799, tangentMode=RCTM_User, time=1, value=1}
       ]}}}
    Lift Contribution=0.333
    Lift Shape Curve=DataInterface() {propertyValues {Curve {keys [
        {interpMode=RCIM_Cubic, tangentMode=RCTM_User, time=0, value=0}
        {interpMode=RCIM_Cubic, tangentMode=RCTM_User, time=0.5, value=1}
        {tangentMode=RCTM_User, time=1, value=0}
       ]}}}
    Rotation Strength:NiagaraFloat=&RandomRangeFloat(Minimum=3, Maximum=8)
    Use Linear Drag=true
    Use Rotational Drag=true
   )
   &SolveForcesAndVelocity(Clamp Velocity=false, Limit Acceleration=false)
   &DynamicMaterialParameters(
    Index 0 Param 1=&MakeCustomFloatFromBool(
     Bool=$User.bIsCurie
     True Float=&FloatFromCurve(FloatCurve=DataInterface() {propertyValues {Curve {keys [
          {arriveTangent=-9.3075, interpMode=RCIM_Cubic, leaveTangent=-9.3075, tangentMode=RCTM_User, time=0, value=2}
          {arriveTangent=-2.0574, interpMode=RCIM_Cubic, leaveTangent=-2.0574, tangentMode=RCTM_User, time=0.2, value=0.625}
          <at 0.75 0 RCIM_Cubic>
         ]}}})
     False Float=0
    )
    Index 0 Param 2=&MakeCustomFloatFromBool(Bool=$User.bIsCurie, True Float=1, False Float=0)
    Module.Param0WriteEnabled=true
    Module.Param1WriteEnabled=true
    Module.Param2WriteEnabled=false
    Module.Param3WriteEnabled=false
   )
  ]
  rendererClasses [&NiagaraMeshRendererProperties]
  renderers [{rendererClass=&NiagaraMeshRendererProperties, rendererIndex=0}]
 }
}
summary {
 systemName=NS_Destruction_Asteria_StealthGrass
 emitters [
  {emitterName=Leaves, simTarget=GPUComputeSim, rendererClasses [&NiagaraMeshRendererProperties]}
  {emitterName=Leaves_Mobile, rendererClasses [&NiagaraMeshRendererProperties]}
  {emitterName=Curie_Sparks, simTarget=GPUComputeSim, rendererClasses [&NiagaraSpriteRendererProperties]}
  {emitterName=Curie_Sparks001, rendererClasses [&NiagaraSpriteRendererProperties]}
 ]
 userVariables [
  User.bIsCurie {type=NiagaraBool}
  User.Bounds {type=Vector3f, description="", defaultValue {struct=Vector3f, value=<3450 3450 3000>}}
  User.DamageDealt {type=NiagaraFloat, defaultValue=1, description=""}
  User.Detached {type=NiagaraBool}
  User.HitLocation {type=NiagaraPosition, description="", defaultValue {}}
  User.HitLocation {type=Vector3f, description="", defaultValue {struct=Vector3f, value=<0 0 0>}}
  User.HitLocation001 {type=NiagaraPosition, description="", defaultValue {}}
  User.HitVector {type=Vector3f, description="", defaultValue {struct=Vector3f, value=<0 0 1>}}
  User.LeafAmount {type=NiagaraFloat, defaultValue=750, description=""}
  User.LeafBaseColor1 {type=LinearColor, description="", defaultValue {struct=LinearColor, value=<rgba 0.0574 0.12 0.012 1>}}
  User.LeafHueVariance {type=NiagaraFloat, defaultValue=-0.02, description=""}
  User.LeafValueVariance {type=NiagaraFloat, defaultValue=0.3, description=""}
  User.NoVariance {type=NiagaraFloat, defaultValue=0, description=""}
  User.Radius {type=NiagaraFloat, defaultValue=100, description=""}
  User.Scale {type=Vector3f, description="", defaultValue {struct=Vector3f, value=<1 1 1>}}
  User.TreeCylinderHeight {type=NiagaraFloat, defaultValue=0, description=""}
  User.TreeCylinderOffset {type=Vector3f, description="", defaultValue {struct=Vector3f, value=<0 0 50>}}
  User.TreeCylinderRadius {type=NiagaraFloat, defaultValue=500, description=""}
  User.TreeSpheriodRadius {type=NiagaraFloat, defaultValue=25, description=""}
 ]
}
