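# Niagara destruction effect for leaf piles, bursting tinted mesh leaves upward from a box sized by User.BoxSize.
# Each burst spawns between half and all of User.LeafAmount leaves, which shrink over their lifetime under gravity, physics-field-aware wind and aerodynamic drag with rotation.
# The GPU emitter enables collision by particle age and cuts wind once a leaf has collided.
# Leafs_Mobile is a near-duplicate without GPU simulation or collision, using constant wind scale.
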NiagaraSystem /Game/Effects/Fort_Effects/World/Destruction/Destruction2/Foliage/NS_Destruction_LeafPiles
dependencies {
 usedDataInterfaces [
  &NiagaraDataInterfaceCurve
  &NiagaraDataInterfaceCollisionQuery
  /Script/ChaosNiagara.NiagaraDataInterfacePhysicsField
  &NiagaraDataInterfaceLandscape
 ]
 usedDynamicInputs []
 usedModules [&SystemState]
}
emitters {
 Leafs {
  simTarget=GPUComputeSim
  emitterUpdateScript [
   &EmitterState()
   &SpawnBurst_Instantaneous(
    Spawn Count=&RandomRangeInt(
     Minimum=/Niagara/DynamicInputs/Divide/Divide_Int(A:NiagaraInt32=&MakeIntFromFloat(Float=$User.LeafAmount), B:NiagaraInt32=2)
     Maximum=&MakeIntFromFloat(Float=$User.LeafAmount)
    )
    Use Spawn Probability=false
    UseLoopCountLimit=false
   )
  ]
  particleSpawnScript [
   &InitializeParticle(
    AdjustAlpha=false
    AdjustHue=true
    AdjustSaturation=true
    AdjustValue=true
    Color=<rgba 0.7758 0.3005 0.0437 1>
    Color Mode=::NewEnumerator3
    Hue Shift Range:Vector2f=<0.005 0.002>
    Lifetime Max=3
    Lifetime Mode=::NewEnumerator1
    Mesh Scale Mode=::NewEnumerator1
    Mesh Uniform Scale Max=5
    Mesh Uniform Scale Min=3
    Saturation Range:Vector2f=<1 2>
    UsePositionOffset=false
    Value Range:Vector2f=<1 2>
    Write Material Random=true
   )
   &ShapeLocation(Box Size=$User.BoxSize, MaskSpawnGroup=false, Override Seed=false, Shape Primitive=::NewEnumerator2, Surface Only=false, UseOwnerScale=true)
   &AddVelocity(Velocity=<0 0 250>)
   &InitialMeshOrientation(Use Rotation Vector=false)
  ]
  particleUpdateScript [
   &ParticleState()
   &ScaleMeshSize(Scale Factor=&VectorFromFloat(Value=&FloatFromCurve(FloatCurve=DataInterface() {propertyValues {Curve {keys [<at 0 1 RCIM_Cubic>, {arriveTangent=-2.2139, interpMode=RCIM_Cubic, leaveTangent=-2.2139, tangentMode=RCTM_User, time=1, value=0}]}}})))
   &GravityForce(Gravity=<0 0 -500>)
   &Collision(
    Age Colliding Particles=false
    Collision Enabled=&SetBoolByFloatComparison(A=$Particles.Age, B=0.05, Comparison Type=::NewEnumerator2)
    Control Roll On Collision=true
    Friction=0.3
    Friction During a Bounce=0.3
    GPU Collision Type=::NewEnumerator2
    Mesh Dimensions:Vector3f=<8 3 0>
    Method for Calculating Particle Radius:ENiagaraCalculateRadiusOptions=::NewEnumerator2
    Module.Control Rest State Via Surface Normal=false
    Module.EnableMaxCollisionCount=false
    Module.Randomize Collision Normal Vector=false
    Radius Calculation Type=::NewEnumerator1
    Restitution=0.3
    ScaleRateOfRotation=false
   )
   &WindForce(
    Ground Mask Elevation Mode:ENiagaraWindGroundMaskMode=::NewEnumerator1
    Use Ground Mask=true
    Use Physics Field=true
    Wind Speed=<1000 0 100>
    Wind Speed Scale=&MakeCustomFloatFromBool(Bool=$StackContext.HasCollided, True Float=0, False Float=0.1)
   )
   &AerodynamicDrag(
    Aerodynamic Drag=0.2
    Aerodynamic Rotational Drag=0.05
    Drag Shape Curve=DataInterface() {propertyValues {Curve {keys [
        {interpMode=RCIM_Cubic, tangentMode=RCTM_User, time=0, value=0.1}
        {arriveTangent=2.0799, interpMode=RCIM_Cubic, leaveTangent=2.0799, tangentMode=RCTM_User, time=1, value=1}
       ]}}}
    Lift Shape Curve=DataInterface() {propertyValues {Curve {keys [
        {interpMode=RCIM_Cubic, tangentMode=RCTM_User, time=0, value=0}
        {interpMode=RCIM_Cubic, tangentMode=RCTM_User, time=0.5, value=1}
        {tangentMode=RCTM_User, time=1, value=0}
       ]}}}
    Rotation Strength:NiagaraFloat=&RandomRangeFloat(Minimum=3, Maximum=8)
    Use Linear Drag=true
    Use Rotational Drag=true
   )
   &SolveForcesAndVelocity(Clamp Velocity=false, Limit Acceleration=false)
  ]
  rendererClasses [&NiagaraMeshRendererProperties]
  renderers [{rendererClass=&NiagaraMeshRendererProperties, rendererIndex=0}]
 }
 Leafs_Mobile {
  emitterUpdateScript [
   &EmitterState()
   &SpawnBurst_Instantaneous(
    Spawn Count=&RandomRangeInt(
     Minimum=&Divide_Int(A:NiagaraInt32=&MakeIntFromFloat(Float=$User.LeafAmount), B:NiagaraInt32=2)
     Maximum=&MakeIntFromFloat(Float=$User.LeafAmount)
    )
    Use Spawn Probability=false
    UseLoopCountLimit=false
   )
  ]
  particleSpawnScript [
   &InitializeParticle(
    AdjustAlpha=false
    AdjustHue=true
    AdjustSaturation=true
    AdjustValue=true
    Color=<rgba 0.7758 0.3005 0.0437 1>
    Color Mode=::NewEnumerator3
    Hue Shift Range:Vector2f=<0.005 0.002>
    Lifetime Max=3
    Lifetime Mode=::NewEnumerator1
    Mesh Scale Mode=::NewEnumerator1
    Mesh Uniform Scale Max=5
    Mesh Uniform Scale Min=3
    Saturation Range:Vector2f=<1 2>
    UsePositionOffset=false
    Value Range:Vector2f=<1 2>
    Write Material Random=true
   )
   &ShapeLocation(Box Size=$User.BoxSize, MaskSpawnGroup=false, Override Seed=false, Shape Primitive=::NewEnumerator2, Surface Only=false, UseOwnerScale=true)
   &AddVelocity(Velocity=<0 0 250>)
   &InitialMeshOrientation(Use Rotation Vector=false)
  ]
  particleUpdateScript [
   &ParticleState()
   &ScaleMeshSize(Scale Factor=&VectorFromFloat(Value=&FloatFromCurve(FloatCurve=DataInterface() {propertyValues {Curve {keys [<at 0 1 RCIM_Cubic>, {arriveTangent=-2.2139, interpMode=RCIM_Cubic, leaveTangent=-2.2139, tangentMode=RCTM_User, time=1, value=0}]}}})))
   &GravityForce(Gravity=<0 0 -500>)
   &WindForce(
    Ground Mask Elevation Mode:ENiagaraWindGroundMaskMode=::NewEnumerator1
    Use Ground Mask=true
    Use Physics Field=true
    Wind Speed=<1000 0 100>
    Wind Speed Scale=0.1
   )
   &AerodynamicDrag(
    Aerodynamic Drag=0.2
    Aerodynamic Rotational Drag=0.05
    Drag Shape Curve=DataInterface() {propertyValues {Curve {keys [
        {interpMode=RCIM_Cubic, tangentMode=RCTM_User, time=0, value=0.1}
        {arriveTangent=2.0799, interpMode=RCIM_Cubic, leaveTangent=2.0799, tangentMode=RCTM_User, time=1, value=1}
       ]}}}
    Lift Contribution=0.333
    Lift Shape Curve=DataInterface() {propertyValues {Curve {keys [
        {interpMode=RCIM_Cubic, tangentMode=RCTM_User, time=0, value=0}
        {interpMode=RCIM_Cubic, tangentMode=RCTM_User, time=0.5, value=1}
        {tangentMode=RCTM_User, time=1, value=0}
       ]}}}
    Rotation Strength:NiagaraFloat=&RandomRangeFloat(Minimum=3, Maximum=8)
    Use Linear Drag=true
    Use Rotational Drag=true
   )
   &SolveForcesAndVelocity(Clamp Velocity=false, Limit Acceleration=false)
  ]
  rendererClasses [&NiagaraMeshRendererProperties]
  renderers [{rendererClass=&NiagaraMeshRendererProperties, rendererIndex=0}]
 }
}
summary {
 systemName=NS_Destruction_LeafPiles
 emitters [
  {emitterName=Leafs, simTarget=GPUComputeSim, rendererClasses [&NiagaraMeshRendererProperties]}
  {emitterName=Leafs_Mobile, rendererClasses [&NiagaraMeshRendererProperties]}
 ]
 userVariables [
  User.BoxSize {type=Vector3f, description="", defaultValue {struct=Vector3f, value=<0 0 0>}}
  User.LeafAmount {type=NiagaraFloat, defaultValue=75, description=""}
 ]
}
